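# Looping positional music cue for the Gargoyle quest, routed to the quest music sound class.
# A quality-level node gives the first two levels the standard wave and the last two the "_opt" wave, feeding a smoothed, high-intensity Doppler node.
# Attenuation overrides add distance falloff and occlusion with a low-pass filter and slow interpolation.
# Concurrency overrides cap simultaneous instances at a few, stopping the quietest.
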SoundCue /Game/Athena/Sounds/Quests/Gargoyle/Quest_Gargoyle_Music_Cue
root=Doppler_0
nodes [
 {id=Doppler_0, children [QualityLevel_0], properties {bUseSmoothing=true, dopplerIntensity=10}}
 {id=QualityLevel_0, children [WavePlayer_5,WavePlayer_5,WavePlayer_1,WavePlayer_1]}
 {id=WavePlayer_1, properties {bLooping=true, soundWaveAssetPtr=&/Quest_Gargoyle_Music_opt}}
 {id=WavePlayer_5, properties {bLooping=true, soundWaveAssetPtr=&/Quest_Gargoyle_Music}}
]
properties {
 bOverrideAttenuation=true
 bOverrideConcurrency=true
 maxDistance=1500
 soundClassObject=/Game/Sounds/FortSoundClassesAndModes/Fort_Music_Quest_PSM
 attenuationOverrides {
  bEnableOcclusion=true
  falloffDistance=1000
  occlusionInterpolationTime=2
  occlusionLowPassFilterFrequency=1000
  occlusionTraceChannel=ECC_GameTraceChannel11
  attenuationShapeExtents {x=500}
 }
 concurrencyOverrides {maxCount=3, resolutionRule=StopQuietest}
}
